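; Blueprint cosmetic actor for the Aerated (GreenAir) vehicle boost, showing its effect while a boost is running.
; The boost started event sets the Boost component active, and the boost finished event sets it inactive, both without a reset.
; The construction script is empty.
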(event Vehicle|Cosmetics|EventOnBoostStarted
  (Components|Activation|SetActive (Variables|Default|GetBoost) true false))

(event Vehicle|Cosmetics|EventOnBoostFinished
  (Components|Activation|SetActive (Variables|Default|GetBoost) false false))

(fn ConstructionScript)
