Material /VehicleCosmetics/Boosts/Boost_GreenAir/Materials/Boost_Green_Air_MAT
graph {
 properties {
  bUsedWithNiagaraRibbons=true
  bUsedWithNiagaraSprites=true
  blendMode=BLEND_Translucent
  pixelDepthOffsetMode=PDOM_Legacy
  twoSided=true
  usedShadingModels=Default Lit
 }
 nodes [
  Constant() {r=100}
  Constant() {r=50}
  MaterialExpressionVertexColor_1 = ParticleColor()
  MaterialFunctionCall(Emissive=@MaterialExpressionVertexColor_1.RGB) {materialFunction=&MF_FX_EmissiveAdjustments}
  MaterialFunctionCall()
  n1 = DynamicParameter() {paramNames [Speed,Offchance,Param3,Param4]}
  n2 = AppendVector(A=@n1.Offchance, B=ScalarParameter() {parameterName=AppendParam})
  n3 = TextureCoordinate()
  n4 = Time()
  n5 = TextureSample(UVs=Panner(Coordinate=Add(A=@n2, B=@n3), Time=Multiply(A=@n1.Speed, B=@n4)) {speedX=-1}) {texture=/DealershipSharedLibrary/FX/Textures/Gradient_Masks/T_VCVFX_Boost_GreenAir_Trail4_E}
  n6 = Constant() {r=10}
  n7 = TextureCoordinate()
  n8 = Multiply(
   A=Multiply(A=Multiply(A=@n5.R, B=@n6), B=@MaterialExpressionVertexColor_1.A)
   B=Power(Base=ComponentMask(None=@n7) {r=true}, Exp=ScalarParameter() {defaultValue=2, parameterName=Powerfade})
  )
  n9 = TextureCoordinate()
  n10 = Power(Base=OneMinus(None=ComponentMask(None=@n9) {r=true}), Exp=ScalarParameter() {defaultValue=1, parameterName=PowerfadeOneMinus})
  Multiply(A=Multiply(A=Multiply(A=@n8, B=@n10), B=Constant() {r=4}), B=CameraPositionWS())
 ]
}
