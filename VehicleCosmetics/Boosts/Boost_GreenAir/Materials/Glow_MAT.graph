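# Translucent, unlit, two-sided glow material for Niagara particles in the GreenAir vehicle boost effect.
# The red channel of a gradient mask texture is multiplied by particle color and fed into the shared FX emissive-adjustments function.
# An opacity chain multiplies that mask by particle alpha, a Fresnel term with a parameterized exponent, and the OpacityA dynamic parameter, allowing per-particle fade control.
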Material /VehicleCosmetics/Boosts/Boost_GreenAir/Materials/Glow_MAT
graph {
 properties {
  bUsedWithNiagaraSprites=true
  blendMode=BLEND_Translucent
  pixelDepthOffsetMode=PDOM_Legacy
  shadingModel=MSM_Unlit
  twoSided=true
  usedShadingModels=Unlit
 }
 nodes [
  Constant()
  Constant()
  MaterialExpressionVertexColor_6 = ParticleColor()
  n1 = TextureSample() {texture=/DealershipSharedLibrary/FX/Textures/Gradient_Masks/T_VCVFX_Boost_GreenAir_Brillo_E}
  MaterialFunctionCall(Emissive=Multiply(A=@MaterialExpressionVertexColor_6.RGB, B=@n1.R)) {materialFunction=&MF_FX_EmissiveAdjustments}
  MaterialFunctionCall()
  n2 = Multiply(A=@n1.R, B=@MaterialExpressionVertexColor_6.A)
  n3 = Fresnel(ExponentIn=ScalarParameter() {parameterName=FresnelExp})
  Multiply(A=Multiply(A=Multiply(A=@n2, B=@n3), B=DynamicParameter() {paramNames [OpacityA,Param2,Param3,Param4]}.OpacityA))
 ]
}
