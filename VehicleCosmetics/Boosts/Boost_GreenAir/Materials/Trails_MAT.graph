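# Translucent unlit, two-sided trail material for Niagara ribbons in the GreenAir vehicle boost effect.
# Three trail textures pan along U, with dynamic parameters driving speed and offset, and static switches select which masks are summed.
# The sum is scaled by particle alpha, scalar gains and dynamic opacity, with power curves on U and its inverse fading both ends.
# Particle color feeds a shared emissive-adjustment function.
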Material /VehicleCosmetics/Boosts/Boost_GreenAir/Materials/Trails_MAT
graph {
 properties {
  bUsedWithNiagaraRibbons=true
  blendMode=BLEND_Translucent
  pixelDepthOffsetMode=PDOM_Legacy
  shadingModel=MSM_Unlit
  twoSided=true
  usedShadingModels=Unlit
 }
 nodes [
  Constant()
  Constant()
  MaterialExpressionVertexColor_4 = ParticleColor()
  MaterialFunctionCall(Emissive=@MaterialExpressionVertexColor_4.RGB) {materialFunction=&MF_FX_EmissiveAdjustments}
  MaterialFunctionCall()
  n1 = DynamicParameter() {paramNames [Velocity,Offset,Opacity,Param4]}
  n2 = Add(A=AppendVector(A=@n1.Offset, B=ScalarParameter()), B=TextureCoordinate())
  n3 = Multiply(A=@n1.Velocity, B=Time())
  n4 = TextureSample(UVs=Panner(Coordinate=@n2, Time=@n3) {speedX=-0.7}) {texture=/DealershipSharedLibrary/FX/Textures/Gradient_Masks/T_VCVFX_Boost_GreenAir_trail1_E}
  n5 = Constant()
  n6 = TextureSample(UVs=Panner(Coordinate=@n2, Time=@n3) {speedX=-0.6}) {texture=/DealershipSharedLibrary/FX/Textures/TexturePackage/T_VCVFX_Boost_GreenAir_Trailcombo2_RGB}
  n7 = Add(
   A=Add(A=StaticSwitchParameter(True=@n4.R, False=@n5) {parameterName=Main}, B=StaticSwitchParameter(True=@n6.R, False=@n5) {parameterName=2Red})
   B=StaticSwitchParameter(True=@n6.G, False=@n5) {parameterName=3Blue}
  )
  n8 = StaticSwitchParameter(True=@n6.B, False=@n5) {parameterName=4Green}
  n9 = TextureSample(UVs=Panner(Coordinate=@n2, Time=@n3) {speedX=-0.6}) {texture=/DealershipSharedLibrary/FX/Textures/Gradient_Masks/T_VCVFX_Boost_GreenAir_Trail3Flipped_E}
  n10 = Multiply(
   A=Add(A=Add(A=@n7, B=@n8), B=StaticSwitchParameter(True=@n9.R, False=@n5) {parameterName=Flipped})
   B=ScalarParameter() {defaultValue=5, parameterName=Opacity5}
  )
  n11 = TextureCoordinate()
  n12 = Multiply(
   A=Multiply(A=Multiply(A=@MaterialExpressionVertexColor_4.A, B=@n10), B=ScalarParameter() {defaultValue=8, parameterName=MultParam})
   B=Power(Base=ComponentMask(None=@n11) {r=true}, Exp=ScalarParameter() {defaultValue=2, parameterName=Fade})
  )
  n13 = TextureCoordinate()
  n14 = Power(Base=OneMinus(None=ComponentMask(None=@n13) {r=true}), Exp=ScalarParameter() {defaultValue=2, parameterName=Fade2})
  Multiply(A=Multiply(A=Multiply(A=@n12, B=@n14), B=@n1.Opacity))
 ]
}
